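# Translucent material for a golf hole prop that fakes a recessed hole through view-dependent UV offsets and depth projection.
# Camera-vector parallax drives planar mask and normal samples, while a second mask projects onto scene geometry via depth reconstruction, minus a custom-depth fade.
# Shading blends scene color toward a dark tint and adds fake half-vector specular from sky light direction and time-of-day color.
# Shading path switches supply constant fallbacks on mobile.
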Material /Hera/Props/Recreation/Golf_Hole/M_GolfHole
graph {
 properties {
  bAutomaticallySetUsageInEditor=false
  blendMode=BLEND_Translucent
  translucencyLightingMode=TLM_VolumetricDirectional
  usedShadingModels=Default Lit
 }
 nodes [
  Constant()
  n1 = NamedRerouteDeclaration(None=AppendVector(A=TextureCoordinate(), B=Constant())) {name=UV, nodeColor=<rgba 0.7109 0 1 1>}
  n2 = CameraVectorWS()
  n3 = Divide(A=ScalarParameter() {defaultValue=-16.0125, group=Main Parameters, parameterName=Depth Offset, sliderMin=0.01}) {constB=100}
  n4 = DotProduct(A=CameraVectorWS(), B=VertexNormalWS())
  n5 = Transform(None=Divide(A=Multiply(A=@n2, B=@n3), B=NamedRerouteDeclaration(None=@n4) {name=Dot Vertex Camera, nodeColor=<rgba 1 0.0297 0 1>})) {transformSourceType=TRANSFORMSOURCE_World, transformType=TRANSFORM_Tangent}
  n6 = NamedRerouteDeclaration(None=ComponentMask(None=Add(A=@n1, B=@n5)) {g=true, r=true}) {name=Planar UV, nodeColor=<rgba 0 1 0.5078 1>}
  NamedRerouteDeclaration(None=TextureSample(UVs=@n6) {samplerSource=SSM_Clamp_WorldGroupSettings, texture=/Hermes/Techart/UtopiaEmissiveSurfaces/Roads/Backdrop_T_WalkWay_mask}.R) {name=Deep Hole, nodeColor=<rgba 0.2891 0 1 1>}
  n7 = TextureSample(UVs=@n6) {
   samplerSource=SSM_Clamp_WorldGroupSettings
   samplerType=SAMPLERTYPE_Normal
   texture=/Hermes/Techart/UtopiaEmissiveSurfaces/Roads/Backdrop_T_WalkWay_Normal
  }
  NamedRerouteDeclaration(None=MaterialFunctionCall(Normal=@n7.RGB, Flatness=ScalarParameter() {defaultValue=5, parameterName=Normal Intensity}) {materialFunction=&FlattenNormal}.Result) {name=Planar Texture, nodeColor=<rgba 0 1 0.5781 1>}
  NamedRerouteUsage()
  QualitySwitch()
  n8 = SkyAtmosphereLightDirection()
  n9 = CameraVectorWS()
  n10 = PixelNormalWS()
  n11 = LinearInterpolate(
   A=ComponentMask(None=SceneTexture() {sceneTextureId=PPI_WorldNormal}.Color) {b=true, g=true, r=true}
   B=Constant3Vector() {constant=<rgba 0 0 1 0>}
  ) {constAlpha=0.8}
  n12 = DotProduct(A=Normalize(VectorInput=Add(A=@n8, B=@n9)), B=Normalize(VectorInput=Add(A=@n10, B=@n11)))
  n13 = ScalarParameter() {defaultValue=10.7375, parameterName=Fake Specular Contrast}
  n14 = Multiply(
   A=Multiply(
    A=MaterialFunctionCall() {materialFunction=&Color_Time_of_Day}.Result
    B=Power(Base=@n12, Exp=@n13)
   )
   B=OneMinus(None=NamedRerouteUsage())
  )
  n15 = VectorParameter() {defaultValue=<rgba 0.0729 0.0729 0.0729 1>}
  n16 = LinearInterpolate(A=SceneColor!(UVs).RGB, B=@n15.RGB, Alpha=NamedRerouteUsage())
  Reroute(None=Add(A=Saturate(None=@n14), B=Multiply(A=@n16, B=@n15.RGB)))
  ShadingPathSwitch(Default=NamedRerouteUsage(), Mobile=Constant3Vector() {constant=<rgba 0 0 1 0>})
  n17 = SceneDepth()
  n18 = PixelDepth()
  n19 = Subtract(
   A=Multiply(A=Divide(A=@n17.R, B=@n18.R), B=WorldPosition() {worldPositionShaderOffset=WPT_CameraRelative}.XYZ)
   B=ActorPositionWS() {originType=CameraRelative}
  )
  n20 = ComponentMask(None=Transform(None=Divide(A=@n19) {constB=100}) {transformSourceType=TRANSFORMSOURCE_World, transformType=TRANSFORM_Local}) {g=true, r=true}
  n21 = TextureSample(UVs=NamedRerouteDeclaration(None=Add(A=@n20) {constB=0.5}) {name=Projected UV, nodeColor=<rgba 0.8281 0 1 1>}) {samplerSource=SSM_Clamp_WorldGroupSettings, texture=&Backdrop_T_WalkWay_mask}
  n22 = Constant() {r=1}
  n23 = Add(
   A=ScalarParameter() {defaultValue=6.2926, group=Demon Rupture, parameterName=Custom Depth Offset}
   B=Subtract(A=PixelDepth().R, B=SceneTexture() {sceneTextureId=PPI_CustomDepth}.Color)
  )
  n24 = Add(
   A=Constant()
   B=Divide(A=ComponentMask(None=@n23) {r=true}, B=ScalarParameter() {defaultValue=6, group=Demon Rupture, parameterName=Custom Depth Gradient})
  )
  ShadingPathSwitch(
   Default=Subtract(
    A=MaterialFunctionCall(In=@n21.RGB, Contrast=@n22) {materialFunction=&CheapContrast}.Result
    B=Saturate(None=@n24)
   )
   Mobile=Constant()
  )
 ]
}
